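# Sound cue for the timer siren of the Creative Floor Siren toy.
# A single wave player feeds Timer_Siren_01 straight to the output, with no randomization or modulation.
# It routes through the Fort_Traps sound class and uses a shared Creative island-wide attenuation asset for spatial falloff.
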SoundCue /Game/Sounds/Creative/Toys/Floor_Siren/Timer_Siren_Cue
root=WavePlayer_0
nodes [{id=WavePlayer_0, properties {soundWaveAssetPtr=&/Timer_Siren_01}}]
properties {
 attenuationSettings=/Game/Sounds/Creative/Attenuations/Fort_Creative_MediumMin_IslandWide_Att
 maxDistance=19000
 soundClassObject=&Fort_Traps
 volumeMultiplier=1
 attenuationOverrides {falloffDistance=1000, attenuationShapeExtents {x=1000}}
}
