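# Sound cue for the timer tick of the Floor Siren creative toy.
# A single wave player feeds the Timer_Tick_01 wave straight to the output, with no randomization or modulation.
# It plays at reduced volume in the Fort_Traps sound class and uses a small-toy focus attenuation to localize it in the world.
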SoundCue /Game/Sounds/Creative/Toys/Floor_Siren/Timer_Tick_Cue
root=WavePlayer_0
nodes [{id=WavePlayer_0, properties {soundWaveAssetPtr=&/Timer_Tick_01}}]
properties {
 attenuationSettings=/Game/Sounds/Creative/Attenuations/Fort_Creative_SmallToy_Focus_Att
 maxDistance=2000
 soundClassObject=&Fort_Traps
 volumeMultiplier=0.6
}
