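# Sound cue for the floor siren toy's minigame end timer tick, judging by its name and folder.
# It plays a single wave player that reuses the front-end start button UI sound at slightly reduced volume.
# It routes to the critical in-game UI sound class with a very large max distance.
# A concurrency override allows one instance at a time and blocks new plays while it runs.
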SoundCue /Game/Sounds/Creative/Toys/Floor_Siren/Timer_UI_Tick_Minigame_End_Cue
root=WavePlayer_2
nodes [{id=WavePlayer_2, properties {soundWaveAssetPtr=/Game/Sounds/Fort_UI_Sounds/FrontEnd/New_FrontEnd/Fortnite_UI_StartButton_01}}]
properties {
 bOverrideConcurrency=true
 maxDistance=2097152
 soundClassObject=&Fort_UI_InGame_Critical
 volumeMultiplier=0.8
 concurrencyOverrides {maxCount=1, resolutionRule=PreventNew}
}
